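# Audio Level Sequence for the DragonCart shock narrative sequence, scoring repeated shock bursts, shard energy and the orb actor.
# Audio sections sit on the Niagara VFX and orb actors, with bursts and shockwaves aligned at three beats and a later heart burst.
# Event tracks trigger on looping ambient sounds, the orb and the shard-powered splines Blueprint.
# Five electricity Niagara components get their ZapVolume parameter keyed to dip, swell, then fade out.
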LevelSequence /DragonCartNarrative/Sequence/DragonCart_Shock/S_DragonCart_Shock_Audio
bindings [
 NS_DragonCart_ShockChargeExample {
  id=#1
  possessed_class=NiagaraActor
  tracks [{type=MovieSceneAudioTrack, display_name=BlastCharge_Energy, sections []}]
 }
 NS_DragonCart_Shard_Electricity_Long_DayOne8 {id=#2, possessed_class=NiagaraActor, tracks []}
 NS_DragonCart_Shard_Electricity_Long_DayOne10 {id=#3, possessed_class=NiagaraActor, tracks []}
 NS_DragonCart_Shard_Electricity_Long_DayOne6 {id=#4, possessed_class=NiagaraActor, tracks []}
 NS_DragonCart_Shard_Electricity_Long_DayOne7 {id=#5, possessed_class=NiagaraActor, tracks []}
 NiagaraComponent0 {
  id=#6
  parent_id=#5
  possessed_class=NiagaraComponent
  tracks [{
    type=MovieSceneNiagaraFloatParameterTrack
    display_name=ZapVolume
    sections [{
      type=MovieSceneFloatSection
      channels [None {type=MovieSceneScriptingFloatChannel, default=1, thinned_from=8, keys [[-4,1],[0,1],[292,0.5],[2288,0.5],[2585,1],[2988,0]]}]
     }]
   }]
 }
 MSS_DragonCart_BeamEnergy_Shock_LP {
  id=#7
  possessed_class=AmbientSound
  tracks [{
    type=MovieSceneEventTrack
    sections [{
      type=MovieSceneEventTriggerSection
      channels [None {type=MovieSceneScriptingEventChannel, const ["<Struct 'MovieSceneEvent' (0x49dde51e380) {}>", 1]}]
     }]
   }]
 }
 NS_DragonCart_ShockExample2 {
  id=#8
  possessed_class=NiagaraActor
  tracks [{
    type=MovieSceneAudioTrack
    display_name=ShockBurst
    sections [
     {
      type=MovieSceneAudioSection
      end_frame=1598
      start_frame=1380
      channels [
       Volume {type=MovieSceneScriptingFloatChannel, default=1}
       Pitch {type=MovieSceneScriptingFloatChannel, default=1}
      ]
     }
     {
      type=MovieSceneAudioSection
      end_frame=2440
      start_frame=2219
      channels [
       Volume {type=MovieSceneScriptingFloatChannel, default=1}
       Pitch {type=MovieSceneScriptingFloatChannel, default=1}
      ]
     }
    ]
   }]
 }
 NS_DragonCart_ShardEnergy_01 {
  id=#9
  possessed_class=NiagaraActor
  tracks [{
    type=MovieSceneAudioTrack
    display_name=Audio
    sections [
     {
      type=MovieSceneAudioSection
      end_frame=2936
      start_frame=2283
      channels [
       Volume {type=MovieSceneScriptingFloatChannel, default=1, keys [[2283,0.7],[2396.955,0.7],[2872.0175,1]]}
       Pitch {type=MovieSceneScriptingFloatChannel, default=1}
      ]
     }
     {
      type=MovieSceneAudioSection
      end_frame=1389
      start_frame=854
      channels [
       Volume {type=MovieSceneScriptingFloatChannel, default=0.5, keys [[854,1],[1143,1],[1287.1175,0.5]]}
       Pitch {type=MovieSceneScriptingFloatChannel, default=1}
      ]
     }
     {
      type=MovieSceneAudioSection
      end_frame=2231
      start_frame=1403
      channels [
       Volume {type=MovieSceneScriptingFloatChannel, default=0.5, keys [[1403,1],[1693,1],[1918.5825,0.5]]}
       Pitch {type=MovieSceneScriptingFloatChannel, default=1}
      ]
     }
     {
      type=MovieSceneAudioSection
      end_frame=2934
      start_frame=2260
      channels [
       Volume {type=MovieSceneScriptingFloatChannel, default=1, keys [[2260.96,1],[2371.5325,1],[2550.64,0.5],[2757.3025,1]]}
       Pitch {type=MovieSceneScriptingFloatChannel, default=1}
      ]
     }
    ]
   }]
 }
 NiagaraComponent0 {
  id=#10
  parent_id=#2
  possessed_class=NiagaraComponent
  tracks [{
    type=MovieSceneNiagaraFloatParameterTrack
    display_name=ZapVolume
    sections [{
      type=MovieSceneFloatSection
      channels [None {type=MovieSceneScriptingFloatChannel, default=1, thinned_from=8, keys [[-4,1],[0,1],[292,0.5],[2288,0.5],[2585,1],[2988,0]]}]
     }]
   }]
 }
 MSS_DragonCart_ShardEnergy_Shock_LP {
  id=#11
  possessed_class=AmbientSound
  tracks [{
    type=MovieSceneEventTrack
    sections [{
      type=MovieSceneEventTriggerSection
      channels [None {type=MovieSceneScriptingEventChannel, const ["<Struct 'MovieSceneEvent' (0x49b280cd160) {}>", 1]}]
     }]
   }]
 }
 NS_DragonCart_OrbBlast_Looping_FromDayOne {id=#12, possessed_class=NiagaraActor, tracks []}
 BP_DragonCart_ShardPowered_Splines {
  id=#13
  possessed_class=BP_DragonCart_ShardPowered_Splines_C
  tracks [{
    type=MovieSceneEventTrack
    sections [{
      type=MovieSceneEventTriggerSection
      channels [None {type=MovieSceneScriptingEventChannel, const ["<Struct 'MovieSceneEvent' (0x49b280c1900) {}>", 1]}]
     }]
   }]
 }
 NiagaraComponent0 {
  id=#14
  parent_id=#15
  possessed_class=NiagaraComponent
  tracks [{
    type=MovieSceneNiagaraFloatParameterTrack
    display_name=ZapVolume
    sections [{
      type=MovieSceneFloatSection
      channels [None {type=MovieSceneScriptingFloatChannel, default=0, keys [[0,1],[296,0.5],[2288,0.5],[2585,1],[2966,0]]}]
     }]
   }]
 }
 NiagaraComponent0 {
  id=#16
  parent_id=#3
  possessed_class=NiagaraComponent
  tracks [{
    type=MovieSceneNiagaraFloatParameterTrack
    display_name=ZapVolume
    sections [{
      type=MovieSceneFloatSection
      channels [None {type=MovieSceneScriptingFloatChannel, default=1, thinned_from=8, keys [[-4,1],[0,1],[292,0.5],[2288,0.5],[2585,1],[2988,0]]}]
     }]
   }]
 }
 BGA_DCN_OrbActorOpen {
  id=#17
  possessed_class=BGA_DCN_OrbActor_C
  tracks [
   {
    type=MovieSceneAudioTrack
    display_name=Audio
    sections [
     {
      type=MovieSceneAudioSection
      end_frame=1023
      start_frame=827
      channels [
       Volume {type=MovieSceneScriptingFloatChannel, default=1}
       Pitch {type=MovieSceneScriptingFloatChannel, default=1}
      ]
     }
     {
      type=MovieSceneAudioSection
      end_frame=2380
      start_frame=2219
      channels [
       Volume {type=MovieSceneScriptingFloatChannel, default=1}
       Pitch {type=MovieSceneScriptingFloatChannel, default=1}
      ]
     }
     {
      type=MovieSceneAudioSection
      end_frame=845
      start_frame=215
      channels [
       Volume {type=MovieSceneScriptingFloatChannel, default=1}
       Pitch {type=MovieSceneScriptingFloatChannel, default=1}
      ]
     }
     {
      type=MovieSceneAudioSection
      end_frame=3183
      start_frame=2686
      channels [
       Volume {type=MovieSceneScriptingFloatChannel, default=1}
       Pitch {type=MovieSceneScriptingFloatChannel, default=1}
      ]
     }
     {
      type=MovieSceneAudioSection
      end_frame=3608
      start_frame=3216
      channels [
       Volume {type=MovieSceneScriptingFloatChannel, default=1}
       Pitch {type=MovieSceneScriptingFloatChannel, default=1}
      ]
     }
     {
      type=MovieSceneAudioSection
      end_frame=1506
      start_frame=1382
      channels [
       Volume {type=MovieSceneScriptingFloatChannel, default=1}
       Pitch {type=MovieSceneScriptingFloatChannel, default=1}
      ]
     }
     {
      type=MovieSceneAudioSection
      end_frame=857
      start_frame=646
      channels [
       Volume {type=MovieSceneScriptingFloatChannel, default=1}
       Pitch {type=MovieSceneScriptingFloatChannel, default=1}
      ]
     }
     {
      type=MovieSceneAudioSection
      end_frame=1414
      start_frame=1151
      channels [
       Volume {type=MovieSceneScriptingFloatChannel, default=1}
       Pitch {type=MovieSceneScriptingFloatChannel, default=1}
      ]
     }
     {
      type=MovieSceneAudioSection
      end_frame=2281
      start_frame=1727
      channels [
       Volume {type=MovieSceneScriptingFloatChannel, default=1}
       Pitch {type=MovieSceneScriptingFloatChannel, default=1}
      ]
     }
     {
      type=MovieSceneAudioSection
      end_frame=2965
      start_frame=2474
      channels [
       Volume {type=MovieSceneScriptingFloatChannel, default=1}
       Pitch {type=MovieSceneScriptingFloatChannel, default=1}
      ]
     }
     {
      type=MovieSceneAudioSection
      end_frame=3516
      start_frame=2994
      channels [
       Volume {type=MovieSceneScriptingFloatChannel, default=1}
       Pitch {type=MovieSceneScriptingFloatChannel, default=1}
      ]
     }
     {
      type=MovieSceneAudioSection
      end_frame=294
      start_frame=0
      channels [
       Volume {type=MovieSceneScriptingFloatChannel, default=1}
       Pitch {type=MovieSceneScriptingFloatChannel, default=1}
      ]
     }
     {
      type=MovieSceneAudioSection
      end_frame=383
      start_frame=0
      channels [
       Volume {type=MovieSceneScriptingFloatChannel, default=0, keys [[0.615,0],[163.615,1]]}
       Pitch {type=MovieSceneScriptingFloatChannel, default=1}
      ]
     }
     {
      type=MovieSceneAudioSection
      end_frame=1388
      start_frame=843
      channels [
       Volume {type=MovieSceneScriptingFloatChannel, default=1}
       Pitch {type=MovieSceneScriptingFloatChannel, default=1}
      ]
     }
     {
      type=MovieSceneAudioSection
      end_frame=2202
      start_frame=1396
      channels [
       Volume {type=MovieSceneScriptingFloatChannel, default=1}
       Pitch {type=MovieSceneScriptingFloatChannel, default=1}
      ]
     }
     {
      type=MovieSceneAudioSection
      end_frame=2877
      start_frame=2226
      channels [
       Volume {type=MovieSceneScriptingFloatChannel, default=0.5, keys [[2226.5225,1],[2378,0.5]]}
       Pitch {type=MovieSceneScriptingFloatChannel, default=1}
      ]
     }
    ]
   }
   {
    type=MovieSceneEventTrack
    display_name=OrbActorEvents
    sections [{
      type=MovieSceneEventTriggerSection
      channels [None {
        type=MovieSceneScriptingEventChannel
        keys [
         [843, "<Struct 'MovieSceneEvent' (0x49e25270280) {}>"]
         [2719, "<Struct 'MovieSceneEvent' (0x49e25270320) {}>"]
         [3243, "<Struct 'MovieSceneEvent' (0x49e252703c0) {}>"]
         [3602, "<Struct 'MovieSceneEvent' (0x49e25270460) {}>"]
        ]
       }]
     }]
   }
  ]
 }
 NS_DragonCart_Shard_Electricity_Long_DayOne9 {id=#15, possessed_class=NiagaraActor, tracks []}
 NS_DragonCart_Shard_Heart_Burst {
  id=#18
  possessed_class=NiagaraActor
  tracks [{
    type=MovieSceneAudioTrack
    display_name=Audio
    sections [{
      type=MovieSceneAudioSection
      end_frame=3320
      start_frame=2929
      channels [
       Volume {type=MovieSceneScriptingFloatChannel, default=1}
       Pitch {type=MovieSceneScriptingFloatChannel, default=1}
      ]
     }]
   }]
 }
 NS_DragonCart_ShardShockWave {
  id=#19
  possessed_class=NiagaraActor
  tracks [{
    type=MovieSceneAudioTrack
    display_name=Audio
    sections [
     {
      type=MovieSceneAudioSection
      end_frame=1049
      start_frame=840
      channels [
       Volume {type=MovieSceneScriptingFloatChannel, default=1}
       Pitch {type=MovieSceneScriptingFloatChannel, default=1}
      ]
     }
     {
      type=MovieSceneAudioSection
      end_frame=2440
      start_frame=2219
      channels [
       Volume {type=MovieSceneScriptingFloatChannel, default=1}
       Pitch {type=MovieSceneScriptingFloatChannel, default=1}
      ]
     }
     {
      type=MovieSceneAudioSection
      end_frame=1598
      start_frame=1380
      channels [
       Volume {type=MovieSceneScriptingFloatChannel, default=1}
       Pitch {type=MovieSceneScriptingFloatChannel, default=1}
      ]
     }
    ]
   }]
 }
 NiagaraComponent0 {
  id=#20
  parent_id=#4
  possessed_class=NiagaraComponent
  tracks [{
    type=MovieSceneNiagaraFloatParameterTrack
    display_name=ZapVolume
    sections [{
      type=MovieSceneFloatSection
      channels [None {type=MovieSceneScriptingFloatChannel, default=1, thinned_from=8, keys [[-4,1],[0,1],[292,0.5],[2288,0.5],[2585,1],[2988,0]]}]
     }]
   }]
 }
 NS_DragonCart_ShockExample {
  id=#21
  possessed_class=NiagaraActor
  tracks [{
    type=MovieSceneAudioTrack
    display_name=ShockBurst
    sections [{
      type=MovieSceneAudioSection
      end_frame=1049
      start_frame=840
      channels [
       Volume {type=MovieSceneScriptingFloatChannel, default=1}
       Pitch {type=MovieSceneScriptingFloatChannel, default=1}
      ]
     }]
   }]
 }
]
display_rate {denominator=1, numerator=60}
tick_resolution {denominator=1, numerator=24000}
tracks []
